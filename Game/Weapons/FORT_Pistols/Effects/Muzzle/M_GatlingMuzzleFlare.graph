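# Unlit translucent material for the gatling gun muzzle flare, usable on mesh particles.
# Particle color RGB feeds the MF_FX_EmissiveAdjustments material function as its emissive input.
# The graph holds only that function call, with no textures of its own, and previews on the gatling muzzle flare mesh.
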Material /Game/Weapons/FORT_Pistols/Effects/Muzzle/M_GatlingMuzzleFlare
graph {
 properties {
  bUsedWithMeshParticles=true
  blendMode=BLEND_Translucent
  pixelDepthOffsetMode=PDOM_Legacy
  previewMesh=&/SM_GatlingGunMuzzleFlare_01
  shadingModel=MSM_Unlit
  usedShadingModels=Unlit
 }
 nodes [MaterialFunctionCall(Emissive=ParticleColor().RGB) {materialFunction=&MF_FX_EmissiveAdjustments}]
}
