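# Unlit translucent material for a minigun barrel glow, set up for skeletal meshes and mesh deformers.
# An orange HDR Emissive Color parameter runs through the shared FX emissive adjustments function.
# A local-space X gradient, offset and raised to a power, is scaled by an Opacity parameter and the color's alpha to fade the glow along the mesh.
# Vertex normals scaled by NormalPushAmount give a small outward push vector, likely to inflate the glow shell.
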Material /Game/Weapons/FORT_Pistols/Effects/Muzzle/M_MinigunBarrelGlow
graph {
 properties {
  bUsedWithMeshDeformer=true
  bUsedWithSkeletalMesh=true
  blendMode=BLEND_Translucent
  pixelDepthOffsetMode=PDOM_Legacy
  shadingModel=MSM_Unlit
  usedShadingModels=Unlit
 }
 nodes [
  n1 = VectorParameter() {parameterName=Emissive Color, defaultValue=<rgba 10 2.0833 0.5 0.75>}
  MaterialFunctionCall(Emissive=@n1.RGB) {materialFunction=&MF_FX_EmissiveAdjustments}
  Multiply(A=VertexNormalWS(), B=ScalarParameter() {defaultValue=0.1, parameterName=NormalPushAmount})
  n2 = MaterialFunctionCall() {materialFunction=&LocalPosition}
  n3 = Add(A=Divide(A=ComponentMask(None=@n2.Local Position) {r=true}) {constB=150}) {constB=0.06}
  Multiply(A=@n1.A, B=Multiply(A=Power(Base=@n3), B=ScalarParameter() {parameterName=Opacity}))
 ]
}
