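# M_Muzzle_Core is an unlit translucent particle sprite material for the core of pistol muzzle flash effects.
# Particle color RGB passes through the MF_FX_EmissiveAdjustments function for emissive adjustment.
# Particle alpha multiplies with the alpha channel of T_MuzzleFlash_Core, so the texture supplies the flash shape while particles drive color and fade.
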Material /Game/Weapons/FORT_Pistols/Effects/Muzzle/M_Muzzle_Core
graph {
 properties {
  bUsedWithParticleSprites=true
  blendMode=BLEND_Translucent
  pixelDepthOffsetMode=PDOM_Legacy
  shadingModel=MSM_Unlit
  usedShadingModels=Unlit
 }
 nodes [
  n1 = ParticleColor()
  MaterialFunctionCall(Emissive=@n1.RGB) {materialFunction=&MF_FX_EmissiveAdjustments}
  Multiply(A=@n1.A, B=TextureSample() {texture=&/T_MuzzleFlash_Core}.A)
 ]
}
