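# Unlit, translucent, two-sided material for muzzle flash mesh particles, drawing a cropped region of a muzzle flash texture.
# Particle color tints the texture, then day/night bloom control and an emissive adjustment function shape the glow.
# Texture alpha is multiplied by a squared near-camera fade and a view-angle falloff with SoftEdge exponent, softening edge-on mesh faces.
# A separate inverted-fresnel branch feeds a quality switch that becomes constant on Low, with no consumer shown.
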Material /Game/Weapons/FORT_Pistols/Effects/Muzzle/M_Muzzle_Flash_Sheet
graph {
 properties {
  bUsedWithMeshParticles=true
  bUsedWithNiagaraMeshParticles=true
  blendMode=BLEND_Translucent
  pixelDepthOffsetMode=PDOM_Legacy
  shadingModel=MSM_Unlit
  twoSided=true
  usedShadingModels=Unlit
 }
 nodes [
  n1 = ParticleColor()
  n2 = VectorParameter() {parameterName=Crop, defaultValue=<rgba 0 0.75 1 1>}
  n3 = TextureSample(UVs=MaterialFunctionCall!(
    UVs=TextureCoordinate()
    TextureIn
    UpperLeftCorner=AppendVector(A=@n2.R, B=@n2.R)
    LowerRightCorner=AppendVector(A=@n2.B, B=@n2.B)
    ExportFloat4
   ) {materialFunction=/Engine/Functions/Engine_MaterialFunctions01/Texturing/TextureCropping}.Crop UVs) {texture=&/T_MuzzleFlash_01}
  MaterialFunctionCall(Emissive=MaterialFunctionCall!(DayBloom_Mult=Constant() {r=0.8}, In=Multiply(A=@n1.RGB, B=@n3.RGB)) {materialFunction=/Game/Packages/Fortnite_UniqueMaterials_1/MH_Materials/DayNight_Auto_Bloom_Control}.Result) {materialFunction=&MF_FX_EmissiveAdjustments}
  n4 = MaterialFunctionCall(Distance=ScalarParameter() {parameterName=NearFadeDistance}) {materialFunction=&Near_Camera_Fade}
  n5 = Add(A=Multiply(A=@n4.Result, B=@n4.Result), B=ScalarParameter() {parameterName=NearFadeAdjustment})
  n6 = DotProduct(A=VertexNormalWS(), B=CameraVectorWS())
  n7 = ScalarParameter() {defaultValue=1, parameterName=SoftEdge}
  Multiply(A=Multiply(A=@n3.A, B=Saturate(None=@n5)), B=Power(Base=Abs(None=@n6), Exp=@n7) {constExponent=4})
  Multiply()
  n8 = MaterialFunctionCall(Invert Fresnel=StaticBool() {value=true}) {materialFunction=&Fresnel_Function}
  n9 = Clamp(None=Multiply(A=Power(Base=@n8.Result, Exp=@n7) {constExponent=4}) {constB=9})
  QualitySwitch(Default=@n9, Low=Constant() {r=1}, High=@n9, Medium=@n9)
 ]
}
