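# Unlit translucent particle material for a soft pistol muzzle flare, usable on sprite particles and Niagara sprites.
# The green channel of a sun glare texture masks particle color into a clamped LDR glow, and also multiplies particle alpha, apparently for opacity.
# A shading path switch expands the glow through DeriveHDRfromLDR by default, while forward and mobile paths use a tinted Mobile Intensity multiply.
# The switched result passes through the shared MF_FX_EmissiveAdjustments function.
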Material /Game/Weapons/FORT_Pistols/Effects/Muzzle/M_Muzzle_Soft_Flae_FFT
graph {
 properties {
  bUsedWithNanite=true
  bUsedWithNiagaraSprites=true
  bUsedWithParticleSprites=true
  blendMode=BLEND_Translucent
  pixelDepthOffsetMode=PDOM_Legacy
  shadingModel=MSM_Unlit
  usedShadingModels=Unlit
 }
 nodes [
  n1 = ParticleColor()
  n2 = TextureSample() {texture=/Game/Effects/Fort_Effects/Textures/T_SUN_Glare_A}
  n3 = Clamp(None=Multiply(A=@n1.RGB, B=@n2.G))
  n4 = Multiply(A=Multiply(A=@n3, B=@n1.RGB), B=ScalarParameter() {defaultValue=10, parameterName=Mobile Intensity}) {constB=100}
  MaterialFunctionCall(Emissive=ShadingPathSwitch(
    Default=MaterialFunctionCall(LDR Input=@n3, HDR Tint=@n1.RGB) {materialFunction=&DeriveHDRfromLDR}.RemappedRange
    Forward=@n4
    Mobile=@n4
   )) {materialFunction=&MF_FX_EmissiveAdjustments}
  Reroute(None=Multiply(A=@n1.A, B=@n2.G))
 ]
}
