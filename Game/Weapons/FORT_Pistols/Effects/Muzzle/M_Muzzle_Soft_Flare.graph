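# Additive unlit particle material for a soft pistol muzzle flare, usable on particle and Niagara sprites.
# A sun-glare texture is tinted by particle color, then blended toward a dimmed copy at night using the alpha channel of the FortniteActiveTimeOfDay collection parameter.
# The result passes through MF_FX_EmissiveAdjustments.
# A shading path switch applies parameterized depth fade by default and a constant one on mobile.
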Material /Game/Weapons/FORT_Pistols/Effects/Muzzle/M_Muzzle_Soft_Flare
graph {
 properties {
  bUsedWithNiagaraSprites=true
  bUsedWithParticleSprites=true
  blendMode=BLEND_Additive
  pixelDepthOffsetMode=PDOM_Legacy
  shadingModel=MSM_Unlit
  usedShadingModels=Unlit
 }
 nodes [
  n1 = Multiply(A=ParticleColor().RGB, B=TextureSample() {texture=/Game/Effects/Fort_Effects/Textures/T_SUN_Glare_A}.RGB)
  n2 = Constant() {r=0.4}
  n3 = CollectionParameter() {
   collection=&FortniteMaterialParameters
   parameterName=FortniteActiveTimeOfDay
  }
  n4 = Constant4Vector() {constant=<rgba 0 0 0 1>}
  MaterialFunctionCall(Emissive=LinearInterpolate(A=@n1, B=Multiply(A=@n2, B=@n1), Alpha=DotProduct(A=@n3, B=@n4.RGBA)) {desc="A=Day B=Night"}) {materialFunction=&MF_FX_EmissiveAdjustments}
  n5 = Constant() {r=1}
  ShadingPathSwitch(Default=DepthFade(Opacity=@n5, FadeDistance=ScalarParameter() {parameterName=DepthFade_Distance}), Mobile=@n5) {desc=Migrated from FeatureLevelSwitch}
 ]
}
