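# Unlit additive particle material for pistol muzzle effects, drawing a soft radial glow on Niagara or Cascade particles.
# Particle color shifts between dimmed and full by a distance blend, dims at night via the time-of-day parameter, then passes through MF_FX_EmissiveAdjustments.
# A radial gradient looks up a soft ramp texture, scaled by particle alpha.
# Depth fade softens intersections, except at Low quality, where a quality switch skips it.
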Material /Game/Weapons/FORT_Pistols/Effects/Muzzle/M_Sphere_depth_bias_01
graph {
 properties {
  bUsedWithNiagaraMeshParticles=true
  bUsedWithNiagaraSprites=true
  bUsedWithParticleSprites=true
  blendMode=BLEND_Additive
  pixelDepthOffsetMode=PDOM_Legacy
  shadingModel=MSM_Unlit
  usedShadingModels=Unlit
 }
 nodes [
  n1 = ParticleColor()
  n2 = LinearInterpolate(
   A=Multiply(A=@n1.RGB) {constB=0.2}
   B=@n1.RGB
   Alpha=MaterialFunctionCall!(Blend Range=Constant() {r=300}, Start Offset=Constant() {r=-400}) {materialFunction=/Engine/Functions/Engine_MaterialFunctions03/Particles/Distance_Blend}.Result
  )
  n3 = CollectionParameter() {
   collection=&FortniteMaterialParameters
   parameterName=FortniteActiveTimeOfDay
  }
  n4 = Constant4Vector() {constant=<rgba 1 1 1 0>}
  MaterialFunctionCall(Emissive=LinearInterpolate(A=Multiply(A=@n2) {constA=2, constB=0.7}, B=@n2, Alpha=DotProduct(A=@n3, B=@n4.RGBA)) {desc="A=Night B=Day"}) {materialFunction=&MF_FX_EmissiveAdjustments}
  n5 = MaterialFunctionCall!(Radius, CenterPosition) {materialFunction=/Engine/Functions/Engine_MaterialFunctions01/Gradient/RadialGradient}
  n6 = Multiply(A=TextureSample(UVs=OneMinus(None=@n5.RadialGradient)) {texture=/Game/Effects/Fort_Effects/Textures/Ramps/T_Ramp_Radial_Soft}.RGB, B=@n1.A)
  n7 = DepthFade(Opacity=@n6, FadeDistance=ScalarParameter() {defaultValue=5, parameterName=Depthfade-Distance}) {fadeDistanceDefault=5}
  QualitySwitch(Default=@n7, Low=@n6, High=@n7, Medium=@n7)
 ]
}
