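# Additive unlit mesh-particle material for a pistol muzzle flash, tinted by particle color and optionally by a color parameter applied to the texture's blue channel.
# Dynamic Param1 blends in a panning tornado pattern distorted by scrolling cloud noise.
# A distance blend scales particle color before the emissive adjustment function.
# Opacity combines texture and particle alpha, an inverted-Fresnel soft edge bypassed at low quality, an optional UV edge mask and depth fade.
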Material /Game/Weapons/FORT_Pistols/Effects/Muzzle/M_muzzle_flash_01_BloomCorrect_BeforeDof.M_muzzle_flash_01_BloomCorrect_BeforeDOF
graph {
 properties {
  bUsedWithMeshParticles=true
  blendMode=BLEND_Additive
  opacityMaskClipValue=0.933
  pixelDepthOffsetMode=PDOM_Legacy
  shadingModel=MSM_Unlit
  usedShadingModels=Unlit
 }
 nodes [
  Constant() {r=1}
  n1 = TextureSample!(
   UVs
   Tex
   Bias=ScalarParameter() {desc=Add this parameter to bias the mip of texutre in order to get a low resolution, parameterName=TextureMipBias}
   Apply View MipBias
  ) {mipValueMode=TMVM_MipBias, texture=&/T_muzzle_flash_01}
  n2 = ParticleColor()
  n3 = MaterialFunctionCall(Invert Fresnel=StaticBool() {value=true}) {materialFunction=&Fresnel_Function}
  n4 = ScalarParameter() {defaultValue=4, parameterName=SoftEdge}
  n5 = Clamp(None=Multiply(A=Power(Base=@n3.Result, Exp=@n4) {constExponent=4}) {constB=9})
  n6 = Multiply(A=Multiply(A=@n1.A, B=@n2.A), B=QualitySwitch(Default=@n5, Low=Constant() {r=1}, High=@n5, Medium=@n5))
  n7 = MaterialFunctionCall(Float2=TextureCoordinate()) {materialFunction=&BreakOutFloat2Components}
  n8 = Saturate(None=Multiply(A=OneMinus(None=@n7.G), B=ScalarParameter() {defaultValue=12, parameterName=SoftCenterEdgeMaskAmount}) {constB=12})
  DepthFade(
   Opacity=StaticSwitchParameter(True=Multiply(A=@n6, B=@n8), False=@n6) {parameterName=SoftCenterEdgeMask}
   FadeDistance=ScalarParameter() {defaultValue=5, parameterName=DepthFadeDist}
  )
  n9 = StaticSwitchParameter(True=Multiply(A=VectorParameter() {parameterName=Muzzle Texture Color, defaultValue=<rgba 1 1 1 0>}.RGB, B=@n1.B), False=@n1.RGB) {defaultValue=true, parameterName=Muzzle Color Control?}
  n10 = Panner(Coordinate=TextureCoordinate()) {speedY=0.16}
  n11 = Multiply(A=ComponentMask(None=TextureSample!(UVs=@n10, Tex, Apply View MipBias) {texture=&/T_Cloud_d}.R) {g=true, r=true}) {constB=0.2}
  n12 = TextureCoordinate()
  n13 = DynamicParameter() {defaultValue=<rgba 0 0 0 1>, paramNames [Param1,Param2,VTiling,UTiling]}
  n14 = Panner(Coordinate=Multiply(A=Add(A=@n11, B=@n12), B=AppendVector(A=@n13.VTiling, B=@n13.VTiling))) {speedY=0.32}
  n15 = Multiply(
   A=@n9
   B=Power(Base=TextureSample(UVs=@n14) {texture=/Game/Global/Textures/Default/Tileable_Patterns/T_Tornado_Tile}.RGB) {constExponent=1.5}
  )
  n16 = Multiply(A=@n2.RGB) {constB=0.3}
  n17 = MaterialFunctionCall!(Blend Range=Constant() {r=300}, Start Offset=Constant() {r=-400}) {materialFunction=/Engine/Functions/Engine_MaterialFunctions03/Particles/Distance_Blend}
  MaterialFunctionCall(Emissive=Multiply(A=LinearInterpolate(A=@n9, B=@n15, Alpha=@n13.Param1), B=LinearInterpolate(A=@n16, B=@n2.RGB, Alpha=@n17.Result))) {materialFunction=&MF_FX_EmissiveAdjustments}
 ]
}
