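# Additive, unlit, two-sided particle material for pistol muzzle tracer effects, usable on sprites and mesh particles.
# A texture supplies color and alpha, with a Colorize switch for a two-color gradient from its blue channel and a blend toward particle color.
# An optional SphereMask clips alpha beyond a dynamic-parameter distance from the muzzle position.
# Color can be divided by eye adaptation before passing through MF_FX_EmissiveAdjustments.
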Material /Game/Weapons/FORT_Pistols/Effects/Muzzle/M_tracer_02
graph {
 properties {
  bUsedWithMeshParticles=true
  bUsedWithNanite=true
  bUsedWithNiagaraMeshParticles=true
  bUsedWithParticleSprites=true
  blendMode=BLEND_Additive
  opacityMaskClipValue=0.933
  pixelDepthOffsetMode=PDOM_Legacy
  shadingModel=MSM_Unlit
  translucencyPass=MTP_BeforeDOF
  twoSided=true
  usedShadingModels=Unlit
 }
 nodes [
  n1 = WorldPosition()
  n2 = DynamicParameter() {parameterIndex=3, paramNames [MuzzlePosX,MuzzlePosY,MuzzlePosZ,MaxDistance]}
  n3 = TextureSampleParameter2D() {parameterName=Texture, texture=&/T_Muzzle_Flash_02}
  n4 = ParticleColor()
  n5 = ScalarParameter() {parameterName=UseParticleColor, sliderMax=1}
  n6 = LinearInterpolate(A=Multiply(A=@n3.A, B=ScalarParameter() {defaultValue=1, parameterName=Opacity}), B=Multiply(A=@n4.A, B=@n3.A), Alpha=@n5)
  LinearInterpolate(
   A=@n6
   B=Multiply(A=@n6, B=SphereMask(A=@n1.XYZ, B=@n2.RGB, Radius=@n2.MaxDistance))
   Alpha=ScalarParameter() {parameterName=UseMaxDistance, sliderMax=1}
  ) {constAlpha=0}
  n7 = LinearInterpolate(
   A=VectorParameter() {parameterName=Color A, defaultValue=<rgba 1 0 0.8267 1>}.RGB
   B=VectorParameter() {parameterName=Color B, defaultValue=<rgba 1 0 0 1>}.RGB
   Alpha=@n3.B
  )
  n8 = LinearInterpolate(A=StaticSwitchParameter(True=Multiply(A=@n7, B=@n3.B), False=@n3.RGB) {parameterName=Colorize}, B=@n4.RGB, Alpha=@n5)
  n9 = EyeAdaptation()
  MaterialFunctionCall(Emissive=LinearInterpolate(A=@n8, B=Divide(A=@n8, B=@n9.EyeAdaptation), Alpha=ScalarParameter() {parameterName=EyeAdaptation, sliderMax=1})) {materialFunction=&MF_FX_EmissiveAdjustments}
 ]
}
